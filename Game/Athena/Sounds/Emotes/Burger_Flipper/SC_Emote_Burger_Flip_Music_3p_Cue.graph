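# Third-person music cue for the Burger Flipper emote, routed to the 3P emote music sound class with emote attenuation and concurrency.
# A mixer layers two looping waves, each shaped by its own enveloper volume curve.
# The music mix holds full volume, then dips gradually, while the fire loop stays silent before ramping above unity as the emote runs on.
# The cue keeps playing when silent and also sends to a music audio-analysis submix.
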SoundCue /Game/Athena/Sounds/Emotes/Burger_Flipper/SC_Emote_Burger_Flip_Music_3p_Cue
root=Mixer_0
nodes [
 {
  id=Enveloper_0
  children [WavePlayer_0]
  properties {pitchCurve {editorCurveData {keys [{time=0}]}}, volumeCurve {editorCurveData {keys [<at 0 1>,<at 40 1>,<at 70 0.75>]}}}
 }
 {
  id=Enveloper_1
  children [WavePlayer_3]
  properties {pitchCurve {editorCurveData {keys [{time=0}]}}, volumeCurve {editorCurveData {keys [<at 0 0>,<at 40 0>,<at 50 1.5>]}}}
 }
 {id=Mixer_0, children [Enveloper_0,Enveloper_1], properties {inputVolume [1,1]}}
 {id=WavePlayer_0, properties {bLooping=true, soundWaveAssetPtr=&/Emote_BurgerFlip_Mix_01}}
 {id=WavePlayer_3, properties {bLooping=true, soundWaveAssetPtr=&/Emote_BurgerFlip_Fire_Loop_01}}
]
properties {
 attenuationSettings=&Fort_Emotes_Athena_Att
 maxDistance=1600
 soundClassObject=/Game/Sounds/FortSoundClassesAndModes/Fort_Music_Emote_3P
 virtualizationMode=PlayWhenSilent
 volumeMultiplier=1.1
 concurrencyOverrides {resolutionRule=PreventNew}
 concurrencySet [/Game/Sounds/Concurrency_Rules/Athena_EmoteMusic_Conc]
 soundSubmixSends [{minSendLevel=0, soundSubmix=&Fortnite_AudioAnalysis_Music}]
}
